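# Animation Blueprint for the face accessory part of the M_MED_Gumball_Dark player character body.
# It copies pose and curves from the attached parent mesh and layers a looping facial idle sequence from the faceAttach bone down.
# The DisableFaceOverride curve blends that result back to the plain copied pose.
# Each update it requests a dynamic material instance for the part's skeletal mesh but does nothing further with it.
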AnimBlueprint /BRCosmetics/Characters/Player/Male/Medium/Bodies/M_MED_Gumball_Dark/Meshes/Parts/M_MED_Gumball_Dark_FaceAcc_AnimBP
target_skeleton=&/M_MED_Gumball_Dark_FaceAcc_Skeleton
graphs [
 AnimGraph {nodes [
   n1 = Copy Pose From Mesh() {bCopyCurves=true, bUseAttachedParent=true}
   n2 = Sequence Player 'Gumball_Dark_Facial_Idle'() {sequence=/BRCosmetics/Animation/Game/MainPlayer/Facial/Gumball/Gumball_Dark_Facial_Idle}
   Output Pose(Result=Two Way Blend!(
     A=Layered blend per bone(BasePose=@n1, BlendPoses_0=@n2) {layerSetup [{branchFilters [{boneName=faceAttach}]}]}
     B=Copy Pose From Mesh() {bCopyCurves=true, bUseAttachedParent=true}
     AlphaCurveName:Name=DisableFaceOverride
    ) -> (Pose) {blendNode {alphaInputType=Curve, alphaBoolBlend {blendOption=Linear}}}) {layerGroup=None}
  ]}
 EventGraph {nodes [
   Event BlueprintUpdateAnimation()
   > Get Dynamic Material Instance(SkeletalMeshComponent=GetPartSkeletalMeshComponent())
  ]}
]
variables []
